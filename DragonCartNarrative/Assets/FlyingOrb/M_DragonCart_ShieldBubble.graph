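# Translucent, unlit, two-sided shield bubble material with a Fresnel-shaped blue glow, edge highlights and index-of-refraction distortion.
# Four wave points radiate micro and macro sine ripples, split by nearest-point masks and warped by panning noise, driving color and opacity.
# Upward-panning wispy noise is masked along UV height and blended into the pattern.
# Custom primitive data supplies Total Opacity and a Hit Time for a timed hit transition.
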Material /DragonCartNarrative/Assets/FlyingOrb/M_DragonCart_ShieldBubble
graph {
 properties {
  bAutomaticallySetUsageInEditor=false
  blendMode=BLEND_Translucent
  refractionMethod=RM_IndexOfRefraction
  shadingModel=MSM_Unlit
  twoSided=true
  usedShadingModels=Unlit
 }
 nodes [
  n1 = ScalarParameter() {defaultValue=1.025, parameterName=Refraction Strength}
  n2 = NamedRerouteUsage()
  n3 = NamedRerouteUsage()
  n4 = LinearInterpolate(B=ScalarParameter() {defaultValue=0.333, parameterName=Hit Fresnel Lower Max}, Alpha=@n3) {constB=0.33}
  n5 = LinearInterpolate(B=ScalarParameter() {defaultValue=3, parameterName=Hit Fresnel Upper Max}, Alpha=@n3) {constA=1, constB=3}
  n6 = Multiply(A=DotProduct(A=CameraVectorWS(), B=VertexNormalWS()), B=TwoSidedSign())
  n7 = OneMinus(None=Saturate(None=InverseLinearInterpolate(A=@n4, B=@n5, Value=@n6)))
  n8 = InverseLinearInterpolate(A=ScalarParameter() {parameterName=Opacity Fresnel Min}, B=ScalarParameter() {defaultValue=1, parameterName=Opacity Fresnel Max}, Value=@n7)
  n9 = LinearInterpolate(
   A=ScalarParameter() {defaultValue=0.5, parameterName=Opacity Fresnel Power Back}
   B=ScalarParameter() {defaultValue=1, parameterName=Opacity Fresnel Power Front}
   Alpha=Saturate(None=TwoSidedSign())
  )
  n10 = NamedRerouteDeclaration(None=ScalarParameter() {parameterName=Hit Percent}) {name=Hit Percent, nodeColor=<rgba 0 1 0.8359 1>}
  n11 = Saturate(None=InverseLinearInterpolate(
    A=ScalarParameter() {defaultValue=0.2, parameterName=Edge Glow Min}
    B=ScalarParameter() {defaultValue=0.8, parameterName=Edge Glow Max}
    Value=@n7
   ))
  n12 = ScalarParameter() {defaultValue=0.6, parameterName=Edge Highlight Min}
  n13 = ScalarParameter() {defaultValue=1, parameterName=Edge Highlight Max}
  n14 = Power(
   Base=Saturate(None=InverseLinearInterpolate(A=@n12, B=@n13, Value=@n7))
   Exp=ScalarParameter() {defaultValue=4, parameterName=Edge Highlight Power}
  )
  n15 = ScalarParameter() {parameterName=Wave Min}
  n16 = ScalarParameter() {defaultValue=1, parameterName=Wave Max}
  n17 = LocalPosition()
  n18 = NamedRerouteUsage()
  n19 = ScalarParameter() {defaultValue=30, parameterName=Distortion Strength}
  n20 = ScalarParameter() {defaultValue=0.5, parameterName=Sphere Distortion Falloff Start Min}
  n21 = ScalarParameter() {defaultValue=1, parameterName=Sphere Distortion Falloff Start Max}
  n22 = VectorParameter() {parameterName=Wave Location 0, defaultValue=<rgba 0.5774 0.5774 0.5774 1>}
  n23 = ScalarParameter() {defaultValue=50, parameterName=Sphere Radius}
  n24 = Multiply(A=Normalize(VectorInput=ComponentMask(None=@n22.RGB) {b=true, g=true, r=true}), B=@n23)
  n25 = Multiply(A=@n23) {constB=2}
  n26 = VectorParameter() {parameterName=Wave Location 1, defaultValue=<rgba 0.5774 -0.5774 -0.5774 1>}
  n27 = Multiply(A=Normalize(VectorInput=ComponentMask(None=@n26.RGB) {b=true, g=true, r=true}), B=@n23)
  n28 = VectorParameter() {parameterName=Wave Location 2, defaultValue=<rgba -0.5774 0.5774 -0.5774 1>}
  n29 = Multiply(A=Normalize(VectorInput=ComponentMask(None=@n28.RGB) {b=true, g=true, r=true}), B=@n23)
  n30 = VectorParameter() {parameterName=Wave Location 3, defaultValue=<rgba -0.5774 -0.5774 0.5774 1>}
  n31 = Multiply(A=Normalize(VectorInput=ComponentMask(None=@n30.RGB) {b=true, g=true, r=true}), B=@n23)
  n32 = Max(
   A=Max(A=SphereMask(A=@n17.XYZ, B=@n24, Radius=@n25) {hardnessPercent=0}, B=SphereMask(A=@n17.XYZ, B=@n27, Radius=@n25) {hardnessPercent=0})
   B=Max(A=SphereMask(A=@n17.XYZ, B=@n29, Radius=@n25) {hardnessPercent=0}, B=SphereMask(A=@n17.XYZ, B=@n31, Radius=@n25) {hardnessPercent=0})
  )
  n33 = OneMinus(None=Saturate(None=InverseLinearInterpolate(A=@n20, B=@n21, Value=@n32) {constA=0.35, constB=0.9}))
  n34 = ScalarParameter() {defaultValue=1, parameterName=Distortion Falloff Power}
  n35 = Add(A=@n17.XYZ, B=Multiply(A=Multiply(A=@n18, B=@n19), B=Power(Base=@n33, Exp=@n34)))
  n36 = Length(None=Subtract(A=@n35, B=@n24)) {is=class MaterialExpressionLength}
  n37 = ScalarParameter() {defaultValue=10, parameterName=Micro Wave Scale}
  n38 = Time()
  n39 = ScalarParameter() {defaultValue=-0.66, parameterName=Micro Wave Speed}
  n40 = Add(A=Divide(A=@n36, B=@n37), B=Multiply(A=@n38, B=@n39))
  n41 = InverseLinearInterpolate(A=@n15, B=@n16, Value=ConstantBiasScale(None=Sine(None=@n40)))
  n42 = ScalarParameter() {defaultValue=200, parameterName=Wave Radius}
  n43 = SphereMask(A=@n35, B=@n24, Radius=@n42) {hardnessPercent=0}
  n44 = SphereMask(A=@n35, B=@n27, Radius=@n42) {hardnessPercent=0}
  n45 = SphereMask(A=@n35, B=@n29, Radius=@n42) {hardnessPercent=0}
  n46 = SphereMask(A=@n35, B=@n31, Radius=@n42) {hardnessPercent=0}
  n47 = Max(A=Max(A=@n43, B=@n44), B=Max(A=@n45, B=@n46))
  n48 = Constant()
  n49 = If(A=@n47, B=@n43, "A > B"=@n48, "A == B"=@n43, "A < B"=@n43)
  n50 = Length(None=Subtract(A=@n35, B=@n27))
  n51 = Time()
  n52 = Add(A=Divide(A=@n50, B=@n37), B=Multiply(A=@n51, B=@n39))
  n53 = InverseLinearInterpolate(A=@n15, B=@n16, Value=ConstantBiasScale(None=Sine(None=@n52)))
  n54 = If(A=@n47, B=@n44, "A > B"=@n48, "A == B"=@n44, "A < B"=@n44)
  n55 = Add(A=Multiply(A=Saturate(None=@n41), B=@n49), B=Multiply(A=Saturate(None=@n53), B=@n54))
  n56 = Length(None=Subtract(A=@n35, B=@n29))
  n57 = Time()
  n58 = Add(A=Divide(A=@n56, B=@n37), B=Multiply(A=@n57, B=@n39))
  n59 = InverseLinearInterpolate(A=@n15, B=@n16, Value=ConstantBiasScale(None=Sine(None=@n58)))
  n60 = If(A=@n47, B=@n45, "A > B"=@n48, "A == B"=@n45, "A < B"=@n45)
  n61 = Length(None=Subtract(A=@n35, B=@n31))
  n62 = Time()
  n63 = Add(A=Divide(A=@n61, B=@n37), B=Multiply(A=@n62, B=@n39))
  n64 = InverseLinearInterpolate(A=@n15, B=@n16, Value=ConstantBiasScale(None=Sine(None=@n63)))
  n65 = If(A=@n47, B=@n46, "A > B"=@n48, "A == B"=@n46, "A < B"=@n46)
  n66 = Add(A=Multiply(A=Saturate(None=@n59), B=@n60), B=Multiply(A=Saturate(None=@n64), B=@n65))
  n67 = TextureCoordinate()
  n68 = NamedRerouteUsage()
  n69 = Panner(
   Coordinate=Add(A=Multiply(A=@n67) {constB=2}, B=Multiply(A=@n68) {constB=0.1})
   Time=Time()
   Speed=AppendVector(A=Constant(), B=ScalarParameter() {defaultValue=0.2, parameterName=Upward Noise Speed})
  )
  n70 = LinearInterpolate(Alpha=NamedRerouteUsage()) {constA=0.25, constB=0}
  n71 = LinearInterpolate(Alpha=NamedRerouteUsage()) {constA=0.5, constB=0.25}
  n72 = ComponentMask(None=TextureCoordinate()) {g=true}
  n73 = Multiply(
   A=LinearInterpolate(
    A=TextureSample(UVs=@n69) {samplerType=SAMPLERTYPE_Grayscale, texture=/Game/Effects/Fort_Effects/Textures/Fire/T_WispyFireNoise_Soft_M}.R
    Alpha=NamedRerouteUsage()
   )
   B=Power(Base=InverseLinearInterpolate(A=@n70, B=@n71, Value=@n72))
  )
  n74 = LinearInterpolate(A=LinearInterpolate(A=Add(A=@n55, B=@n66), Alpha=NamedRerouteUsage()) {constB=0.5}, B=@n73)
  n75 = Multiply(
   A=Power(Base=Saturate(None=@n8), Exp=LinearInterpolate(A=@n9, Alpha=@n10) {constB=0.25})
   B=Max(A=Max(A=@n11, B=@n14), B=LinearInterpolate(Alpha=@n74) {constA=0.25})
  )
  n76 = ScalarParameter() {bUseCustomPrimitiveData=true, defaultValue=1, parameterName=Total Opacity, primitiveDataIndex=1}
  LinearInterpolate(B=LinearInterpolate(B=LinearInterpolate(B=@n1, Alpha=@n2) {constA=1}, Alpha=@n75) {constA=1}, Alpha=@n76) {constA=1}
  n77 = VectorParameter() {parameterName=Sphere Color Low, defaultValue=<rgba 0 0.1534 0.4938 1>}
  n78 = ScalarParameter() {defaultValue=75, parameterName=Macro Wave Scale}
  n79 = ScalarParameter() {defaultValue=-0.5, parameterName=Macro Wave Speed}
  n80 = Sine(None=Add(A=Divide(A=@n36, B=@n78), B=Multiply(A=@n38, B=@n79)))
  n81 = Sine(None=Add(A=Divide(A=@n50, B=@n78), B=Multiply(A=@n51, B=@n79)))
  n82 = Add(A=Multiply(A=ConstantBiasScale(None=@n80), B=@n49), B=Multiply(A=ConstantBiasScale(None=@n81), B=@n54))
  n83 = Sine(None=Add(A=Divide(A=@n56, B=@n78), B=Multiply(A=@n57, B=@n79)))
  n84 = Sine(None=Add(A=Divide(A=@n61, B=@n78), B=Multiply(A=@n62, B=@n79)))
  n85 = Add(A=Multiply(A=ConstantBiasScale(None=@n83), B=@n60), B=Multiply(A=ConstantBiasScale(None=@n84), B=@n65))
  n86 = LinearInterpolate(
   A=@n77.RGB
   B=LinearInterpolate(A=@n77.RGB, B=VectorParameter() {parameterName=Sphere Color High, defaultValue=<rgba 0.073 0.3611 1 1>}.RGB, Alpha=@n74)
   Alpha=LinearInterpolate(A=Add(A=@n82, B=@n85), Alpha=NamedRerouteUsage()) {constB=0.5}
  )
  n87 = VectorParameter() {parameterName=Sphere Color Dot, defaultValue=<rgba 0.326 0.5355 1 1>}
  n88 = TwoSidedSign()
  Multiply(
   A=LinearInterpolate(
    A=LinearInterpolate(A=@n86, B=@n87.RGB, Alpha=@n11)
    B=VectorParameter() {parameterName=Sphere Edge Dot, defaultValue=<rgba 1.2552 3.19 5 1>}.RGB
    Alpha=@n14
   )
   B=LinearInterpolate(A=ScalarParameter() {defaultValue=0.5, parameterName=Backface Brightness}, Alpha=Saturate(None=@n88)) {constA=1}
  )
  n89 = Saturate(None=@n75)
  n90 = NamedRerouteUsage()
  n91 = LinearInterpolate(
   A=ScalarParameter() {defaultValue=0.5, parameterName=Min Upward Noise Hit Opacity}
   B=ScalarParameter() {defaultValue=1, parameterName=Max Upward Noise Hit Opacity}
   Alpha=NamedRerouteUsage()
  )
  Multiply(A=LinearInterpolate(A=Multiply(A=@n89, B=@n90), B=Multiply(A=@n73, B=@n91)), B=@n76)
  n92 = TextureCoordinate()
  n93 = ComponentMask(None=@n92) {r=true}
  n94 = ScalarParameter() {defaultValue=1, parameterName=Distortion Tiling}
  n95 = ComponentMask(None=@n92) {g=true}
  n96 = Panner(
   Coordinate=AppendVector(A=Multiply(A=@n93, B=@n94), B=Multiply(A=@n95, B=@n94))
   Time=Time()
   Speed=ScalarParameter() {defaultValue=0.066, parameterName=Distortion Speed}
  ) {speedX=0.2, speedY=0.2}
  NamedRerouteDeclaration(None=ConstantBiasScale(None=TextureSample(UVs=@n96) {samplerType=SAMPLERTYPE_Grayscale, texture=/FortWater/Textures/LowResBlurredNoise}.R) {bias=-0.5, scale=2}) {name=Distortion Noise, nodeColor=<rgba 1 0 0.375 1>}
  NamedRerouteDeclaration(None=LinearInterpolate(
    A=ScalarParameter() {defaultValue=0.1, parameterName=Min Opacity}
    B=ScalarParameter() {defaultValue=1, parameterName=Max Opacity}
    Alpha=NamedRerouteUsage()
   ) {constA=0.1}) {name=Hit Opacity, nodeColor=<rgba 0 1 0.5078 1>}
  n97 = Divide(
   A=Subtract(A=Time(), B=ScalarParameter() {bUseCustomPrimitiveData=true, parameterName=Hit Time})
   B=ScalarParameter() {defaultValue=0.5, parameterName=Hit Transition Speed}
  ) {constB=4}
  NamedRerouteDeclaration(None=OneMinus(None=Saturate(None=@n97))) {name=Hit Transition, nodeColor=<rgba 1 0.1875 0 1>}
  n98 = InverseLinearInterpolate(Value=@n72) {constA=0.8, constB=0.975}
  NamedRerouteDeclaration(None=Power(Base=Saturate(None=@n98))) {name=Bottom Cap, nodeColor=<rgba 0 0.4141 1 1>}
  n99 = NamedRerouteUsage()
  n100 = ScalarParameter() {defaultValue=1, parameterName=WPO Strength}
  Reroute(None=Multiply(A=VertexNormalWS(), B=Multiply(A=@n99, B=@n100)))
  Subtract(A=@n35, B=@n24)
 ]
}
