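# Material function for the flying orb's cube-of-cubes look, blending glowing glyph panels onto input Material Attributes through a panel texture mask and Alpha input.
# Emissive, scaled by AO, lerps from a Fresnel- and cloud-noise-modulated base to glyph color.
# That lerp's per-channel alpha combines glow color, sine-pulsed intensity and a thresholded height-mask glyph faded by a panning sample.
# An orb/rift parameter collection drives emissive strength and attack-mode color swaps.
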MaterialFunction /DragonCartNarrative/Assets/FlyingOrb/Materials/FlyingOrbCubeofCubes
graph {nodes [
  n1 = TextureCoordinate() {coordinateIndex=1}
  n2 = TextureCoordinate() {coordinateIndex=1}
  n3 = TextureSample(UVs=Multiply(A=@n2, B=ScalarParameter() {defaultValue=0.5, parameterName=CubedetailTiling})) {texture=&/LightsFlyingOrb_M}
  Add(A=@n1, B=@n3.B)
  n4 = ScalarParameter() {defaultValue=2, desc=5, group=T H E C U B E, parameterName=Overall Texture Tiling}
  n5 = Multiply(A=@n4, B=Multiply(A=Constant() {desc=Glyph Panner Independent Scale, r=0.125}, B=@n1))
  Add(A=@n5, B=@n3.B)
  n6 = TextureSample(UVs=Multiply(A=@n2, B=ScalarParameter() {defaultValue=1, parameterName=CubePanels})) {texture=&LightsFlyingOrb_M}
  n7 = FunctionInput() {bUsePreviewValueAsDefault=true, inputName=Alpha, inputType=FunctionInput_Scalar, sortPriority=2}
  n8 = FunctionInput() {inputType=FunctionInput_MaterialAttributes, sortPriority=1}
  n9 = LinearInterpolate(
   A=VectorParameter() {parameterName=Emissivebase, defaultValue=<rgba 30 8.8847 0 1>}.RGB
   B=VectorParameter() {group=Attack Mode, parameterName=AttackMode_EmissiveBase, defaultValue=<rgba 1 0 0 1>}.RGB
   Alpha=CollectionParameter() {collection=/DragonCartNarrative/Assets/VFX/RiftPortalIdle/MPC_DragonCart_Orb_Rift, parameterName=Orb Attack Mode}
  )
  n10 = Add(
   A=CollectionParameter() {collection=&MPC_DragonCart_Orb_Rift, parameterName=Orb Emissive Strength}
   B=CollectionParameter() {collection=&MPC_DragonCart_Orb_Rift, parameterName=Orb Additional Emissive Strength}
  )
  n11 = Fresnel(ExponentIn=ScalarParameter() {defaultValue=0.02, parameterName=Fresnel01}) {exponent=1}
  n12 = Multiply(A=TextureCoordinate(), B=ScalarParameter() {defaultValue=2.7528, parameterName=CloudTiling})
  n13 = Multiply(
   A=Multiply(A=Multiply(A=@n9, B=@n10) {constB=0}, B=OneMinus(None=@n11))
   B=LinearInterpolate(Alpha=TextureSample(UVs=@n12) {texture=&AtmosphericCloudNoise01}.R) {constA=0.5}
  )
  n14 = Multiply(
   A=LinearInterpolate(
    A=VectorParameter() {parameterName=Glyph Color, defaultValue=<rgba 30 8.8847 0 1>}.RGB
    B=VectorParameter() {group=Attack Mode, parameterName=AttackMode_GlyphColor, defaultValue=<rgba 1 0 0 1>}.RGB
    Alpha=CollectionParameter() {collection=&MPC_DragonCart_Orb_Rift, parameterName=Orb Attack Mode}
   )
   B=Add(
    A=CollectionParameter() {collection=&MPC_DragonCart_Orb_Rift, parameterName=Orb Emissive Strength}
    B=CollectionParameter() {collection=&MPC_DragonCart_Orb_Rift, parameterName=Orb Additional Emissive Strength}
   )
  ) {constB=0}
  n15 = VectorParameter() {group=T H E C U B E, parameterName=Glyph Glow Color, defaultValue=<rgba 0.8958 0.115 0 1>}
  n16 = VectorParameter() {group=Attack Mode, parameterName=AttackMode_GlyphGlowColor, defaultValue=<rgba 1 0 0 1>}
  n17 = CollectionParameter() {collection=&MPC_DragonCart_Orb_Rift, parameterName=Orb Attack Mode}
  n18 = Time()
  n19 = Constant() {desc=Glyph Pulse Speed, r=0.3}
  n20 = LinearInterpolate(
   A=ScalarParameter() {defaultValue=12.135, group=T H E C U B E, parameterName=Glpyh Pulse Flare Width Max}
   B=ScalarParameter() {defaultValue=58.9586, group=T H E C U B E, parameterName=Glpyh Pulse Flare Width Min}
   Alpha=MaterialFunctionCall(Value=Multiply(A=@n18, B=@n19)) {materialFunction=&LinearSine}."Rounded Linear Sine "
  )
  n21 = Multiply(A=@n20, B=Constant() {r=3})
  n22 = Multiply(A=Constant() {r=2}, B=@n1)
  n23 = Constant() {desc=Glyph Panner Opacity Max, r=1}
  n24 = ScalarParameter() {group=T H E C U B E, parameterName=Glyph Panner Opacity Min}
  n25 = Constant() {desc="Glyph Panner Speed ", r=0.1}
  n26 = Time()
  n27 = TextureSample!(UVs=Panner(Coordinate=@n5, Time=Multiply(A=@n25, B=@n26)) {speedY=1}, Tex, Bias=Constant() {r=2}, Apply View MipBias) {
   mipValueMode=TMVM_MipBias
   samplerSource=SSM_Wrap_WorldGroupSettings
   samplerType=SAMPLERTYPE_LinearColor
   texture=/Game/Athena/Prototype/Blueprints/Cube/Materials/DFHeightMask
  }
  n28 = Multiply(
   A=TextureSample!(UVs=Multiply(A=@n4, B=@n22), Tex, Apply View MipBias) {samplerType=SAMPLERTYPE_LinearColor, texture=&DFHeightMask}.R
   B=Saturate(None=LinearInterpolate(A=@n23, B=@n24, Alpha=@n27.B))
  )
  n29 = Constant() {r=1.25}
  n30 = Subtract(A=Saturate(None=Multiply(A=@n28, B=@n29)), B=ScalarParameter() {defaultValue=0.5274, parameterName=Thickness})
  n31 = Saturate(None=Multiply(A=LinearInterpolate(A=@n15.RGB, B=@n16.RGB, Alpha=@n17), B=Multiply(A=@n21, B=@n30)))
  n32 = GetMaterialAttributes(None=@n8) {attributeGetTypes [#1, #2]}
  n33 = SetMaterialAttributes!(
   MaterialAttributes=@n32.MaterialAttributes
   BaseColor=Constant()
   Specular=Constant() {r=0.4}
   Metallic=Constant() {desc=Edge Metallic}
   Roughness=Constant() {r=0.2}
   EmissiveColor=Multiply(A=FunctionInput() {inputName=AO}, B=LinearInterpolate(A=@n13, B=@n14, Alpha=@n31))
  ) {attributeSetTypes [
    #3
    #4
    #5
    #6
    #2
   ]}
  FunctionOutput(None=BlendMaterialAttributes(A=@n8, B=@n33, Alpha=Multiply(A=@n6.R, B=@n7)))
  LinearInterpolate(A=TextureCoordinate(), B=@n1, Alpha=Round(None=ScalarParameter() {group=T H E C U B E, parameterName=Use 2nd UV Channel, sortPriority=250}))
  Multiply(A=@n3.B)
  OneMinus(None=@n6.R)
  Power(Base=StaticSwitch(), Exp=@n20)
  Reroute(None=@n30)
  Reroute(None=@n32.Normal)
  StaticSwitch(False=@n27.B)
 ]}
