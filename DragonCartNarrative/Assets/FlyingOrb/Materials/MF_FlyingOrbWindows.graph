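# Material function that adds animated, glowing window lights to the DragonCart flying orb's material.
# Two samples of a tiled lights texture pan in quantized time steps; G times powered B forms the flickering pattern.
# The orb rift parameter collection blends the color toward attack mode and sets emissive strength.
# The AO-scaled result replaces incoming emissive, blended by the Alpha input; several leftover nodes are unconnected.
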MaterialFunction /DragonCartNarrative/Assets/FlyingOrb/Materials/MF_FlyingOrbWindows
graph {nodes [
  Add()
  n1 = GetMaterialAttributes(None=FunctionInput() {inputName=Mat, inputType=FunctionInput_MaterialAttributes}) {attributeGetTypes [#1, #2]}
  n2 = VectorParameter() {parameterName=Window Light Color, defaultValue=<rgba 20 7.3896 0 1>}
  n3 = VectorParameter() {group=Attack Mode, parameterName=AttackMode_WindowLightColor, defaultValue=<rgba 20 0.4167 0.4167 1>}
  n4 = CollectionParameter() {collection=/DragonCartNarrative/Assets/VFX/RiftPortalIdle/MPC_DragonCart_Orb_Rift, parameterName=Orb Attack Mode}
  n5 = CollectionParameter() {collection=&MPC_DragonCart_Orb_Rift, parameterName=Orb Emissive Strength}
  n6 = CollectionParameter() {collection=&MPC_DragonCart_Orb_Rift, parameterName=Orb Additional Emissive Strength}
  n7 = Multiply(A=TextureCoordinate(), B=ScalarParameter() {defaultValue=16, parameterName=Window Tiling})
  n8 = Multiply(A=Time(), B=ScalarParameter() {defaultValue=0.15, parameterName=Speed})
  n9 = ScalarParameter() {defaultValue=8, parameterName=Tiling}
  n10 = Divide(A=Round(None=Multiply(A=@n8, B=@n9) {constB=10}), B=@n9)
  n11 = TextureSample(UVs=Panner(Coordinate=@n7, Time=Frac(None=@n10)) {bFractionalPart=true, speedY=-1}) {texture=&/LightsFlyingOrb_M}
  n12 = Multiply(A=Time(), B=ScalarParameter() {defaultValue=-0.01, parameterName=Speed2})
  n13 = ScalarParameter() {defaultValue=8, parameterName=Tiling}
  n14 = Divide(A=Round(None=Multiply(A=@n12, B=@n13) {constB=10}), B=@n13)
  n15 = TextureSample(UVs=Panner(Coordinate=@n7, Time=Frac(None=@n14)) {bFractionalPart=true, speedX=1, speedY=1}) {texture=&LightsFlyingOrb_M}
  n16 = ScalarParameter() {defaultValue=4, parameterName=Randompower}
  n17 = Multiply(
   A=Multiply(A=LinearInterpolate(A=@n2.RGB, B=@n3.RGB, Alpha=@n4), B=Add(A=@n5, B=@n6)) {constB=0}
   B=Multiply(A=@n11.G, B=Power(Base=@n15.B, Exp=@n16))
  )
  Add(A=@n1.EmissiveColor, B=@n17)
  Ceil(None=Add(A=Subtract(A=@n15.R), B=@n11.G))
  n18 = ScalarParameter() {defaultValue=10, parameterName=Tiling}
  Divide(A=Round(None=Multiply(A=@n11.G, B=@n18) {constB=10}), B=@n18)
  n19 = Multiply(A=@n17, B=FunctionInput() {inputName=AO, inputType=FunctionInput_Scalar})
  FunctionOutput(None=BlendMaterialAttributes(
    A=@n1.MaterialAttributes
    B=SetMaterialAttributes!(MaterialAttributes=@n1.MaterialAttributes, EmissiveColor=@n19, BaseColor) {attributeSetTypes [#1, #2]}
    Alpha=FunctionInput() {inputName=Alpha, inputType=FunctionInput_Scalar}
   ))
  MaterialFunctionCall() {materialFunction=&LinearGradient}
  Multiply(A=@n11.G)
  Multiply(A=@n11.G)
 ]}
