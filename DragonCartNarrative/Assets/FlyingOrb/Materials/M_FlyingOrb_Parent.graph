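# Two-sided parent material for the DragonCart Flying Orb, with textured metal, tiled panel detail and mode-driven glow.
# An FX mask routes panel blending and the CubeofCubes and Windows layer functions, followed by a circular portal opacity mask.
# MPC_DragonCart_Orb_Rift drives emissive strength, attack and pulsing hostile colors, portal mask and sine-wobble WPO scale.
# The rotate-about-axis and corrected-normal branches feed no material attribute.
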Material /DragonCartNarrative/Assets/FlyingOrb/Materials/M_FlyingOrb_Parent
graph {
 expressions {
  MaterialExpressionActorPositionWS_0 {class=MaterialExpressionActorPositionWS, inputs [], properties {}}
  MaterialExpressionAdd_0 {
   class=MaterialExpressionAdd
   inputs [
    A {from=MaterialExpressionMultiply_4}
    B {from=""}
   ]
  }
  MaterialExpressionAdd_1 {
   class=MaterialExpressionAdd
   inputs [
    A {from=MaterialExpressionScalarParameter_8}
    B {from=""}
   ]
   properties {constB=200}
  }
  MaterialExpressionAdd_2 {
   class=MaterialExpressionAdd
   inputs [
    A {from=MaterialExpressionConvert_0}
    B {from=MaterialExpressionRotateAboutAxis_0}
   ]
  }
  MaterialExpressionAdd_3 {
   class=MaterialExpressionAdd
   inputs [
    A {from=MaterialExpressionConstantBiasScale_1}
    B {from=MaterialExpressionComponentMask_1}
   ]
  }
  MaterialExpressionAdd_5 {
   class=MaterialExpressionAdd
   inputs [
    A {from=MaterialExpressionCollectionParameter_4}
    B {from=MaterialExpressionCollectionParameter_6}
   ]
  }
  MaterialExpressionCollectionParameter_0 {
   class=MaterialExpressionCollectionParameter
   inputs []
   properties {collection=/DragonCartNarrative/Assets/VFX/RiftPortalIdle/MPC_DragonCart_Orb_Rift, parameterName=OrbWPOScale}
  }
  MaterialExpressionCollectionParameter_1 {
   class=MaterialExpressionCollectionParameter
   inputs []
   properties {collection=&MPC_DragonCart_Orb_Rift, parameterName=PortalOpen}
  }
  MaterialExpressionCollectionParameter_2 {
   class=MaterialExpressionCollectionParameter
   inputs []
   properties {collection=&MPC_DragonCart_Orb_Rift, parameterName=PortalPosition}
  }
  MaterialExpressionCollectionParameter_3 {
   class=MaterialExpressionCollectionParameter
   inputs []
   properties {collection=&MPC_DragonCart_Orb_Rift, parameterName=Orb Attack Mode}
  }
  MaterialExpressionCollectionParameter_4 {
   class=MaterialExpressionCollectionParameter
   inputs []
   properties {collection=&MPC_DragonCart_Orb_Rift, parameterName=Orb Emissive Strength}
  }
  MaterialExpressionCollectionParameter_5 {
   class=MaterialExpressionCollectionParameter
   inputs []
   properties {collection=&MPC_DragonCart_Orb_Rift, parameterName=Orb Hostile Mode}
  }
  MaterialExpressionCollectionParameter_6 {
   class=MaterialExpressionCollectionParameter
   inputs []
   properties {collection=&MPC_DragonCart_Orb_Rift, parameterName=Orb Additional Emissive Strength}
  }
  MaterialExpressionComponentMask_0 {class=MaterialExpressionComponentMask, inputs [None {from=MaterialExpressionCollectionParameter_2}], properties {b=true, g=true, r=true}}
  MaterialExpressionComponentMask_1 {class=MaterialExpressionComponentMask, inputs [None {from=MaterialExpressionTextureCoordinate_2}], properties {g=true}}
  MaterialExpressionConstant3Vector_0 {class=MaterialExpressionConstant3Vector, properties {constant=<rgba 0 0 1 0>}}
  MaterialExpressionConstant3Vector_1 {class=MaterialExpressionConstant3Vector, properties {constant=<rgba 0 0 -1 0>}}
  MaterialExpressionConstantBiasScale_0 {class=MaterialExpressionConstantBiasScale, inputs [None {from=MaterialExpressionMultiply_4}], properties {scale=1}}
  MaterialExpressionConstantBiasScale_1 {class=MaterialExpressionConstantBiasScale, inputs [None {from=MaterialExpressionSine_2}]}
  MaterialExpressionConstant_0 {class=MaterialExpressionConstant}
  MaterialExpressionConstant_1 {class=MaterialExpressionConstant, properties {r=2}}
  MaterialExpressionConvert_0 {
   class=MaterialExpressionConvert
   inputs [
    None {from=""}
    None {from=MaterialExpressionMultiply_6}
    None {from=MaterialExpressionMultiply_5}
   ]
   properties {
    nodeName=Make Vector3
    convertInputs [
     {type=Scalar, expressionInput {expression=null, inputName=None, mask=0, maskA=0, maskB=0, maskG=0, maskR=0, outputIndex=0}}
     {
      type=Scalar
      expressionInput {
       expression=&self:MaterialExpressionMultiply_6
       inputName=None
       mask=0
       maskA=0
       maskB=0
       maskG=0
       maskR=0
       outputIndex=0
      }
     }
     {
      type=Scalar
      expressionInput {
       expression=&self:MaterialExpressionMultiply_5
       inputName=None
       mask=0
       maskA=0
       maskB=0
       maskG=0
       maskR=0
       outputIndex=0
      }
     }
    ]
    convertOutputs [{type=Vector3}]
   }
  }
  MaterialExpressionDivide_0 {
   class=MaterialExpressionDivide
   inputs [
    A {from=MaterialExpressionFmod_0}
    B {from=MaterialExpressionScalarParameter_9}
   ]
  }
  MaterialExpressionFloor_0 {class=MaterialExpressionFloor, inputs [None {from=MaterialExpressionAdd_3, output=""}], properties {}}
  MaterialExpressionFmod_0 {
   class=MaterialExpressionFmod
   inputs [
    A {from=MaterialExpressionTime_1, output=""}
    B {from=MaterialExpressionScalarParameter_9, output=""}
   ]
   properties {}
  }
  MaterialExpressionFresnel_0 {
   class=MaterialExpressionFresnel
   inputs [
    ExponentIn {from=MaterialExpressionScalarParameter_0, output=""}
    BaseReflectFractionIn {from="", output=""}
    Normal {from="", output=""}
   ]
   properties {baseReflectFraction=0}
  }
  MaterialExpressionLinearInterpolate_0 {
   class=MaterialExpressionLinearInterpolate
   inputs [
    A {from=""}
    B {from=MaterialExpressionLinearInterpolate_9}
    Alpha {from=MaterialExpressionSaturate_0}
   ]
  }
  MaterialExpressionLinearInterpolate_1 {
   class=MaterialExpressionLinearInterpolate
   inputs [
    A {from=MaterialExpressionLinearInterpolate_3}
    B {from=MaterialExpressionMultiply_12}
    Alpha {from=MaterialExpressionTextureSampleParameter2D_5, output=G}
   ]
  }
  MaterialExpressionLinearInterpolate_10 {
   class=MaterialExpressionLinearInterpolate
   inputs [
    A {from=MaterialExpressionVectorParameter_0, output=RGB}
    B {from=MaterialExpressionLinearInterpolate_11}
    Alpha {from=MaterialExpressionCollectionParameter_5}
   ]
  }
  MaterialExpressionLinearInterpolate_11 {
   class=MaterialExpressionLinearInterpolate
   inputs [
    A {from=MaterialExpressionMultiply_18}
    B {from=MaterialExpressionVectorParameter_2, output=RGB}
    Alpha {from=MaterialExpressionConstantBiasScale_1}
   ]
  }
  MaterialExpressionLinearInterpolate_2 {
   class=MaterialExpressionLinearInterpolate
   inputs [
    A {from=MaterialExpressionTextureSampleParameter2D_0, output=RGB}
    B {from=MaterialExpressionMaterialFunctionCall_4, output=Result}
    Alpha {from=MaterialExpressionTextureSampleParameter2D_5, output=G}
   ]
  }
  MaterialExpressionLinearInterpolate_3 {
   class=MaterialExpressionLinearInterpolate
   inputs [
    A {from=MaterialExpressionTextureSampleParameter2D_3, output=RGB}
    B {from=MaterialExpressionConstant_0}
    Alpha {from=MaterialExpressionTextureSampleParameter2D_5, output=R}
   ]
  }
  MaterialExpressionLinearInterpolate_4 {
   class=MaterialExpressionLinearInterpolate
   inputs [
    A {from=MaterialExpressionMultiply_0}
    B {from=""}
    Alpha {from=MaterialExpressionTextureSampleParameter2D_5, output=R}
   ]
   properties {constB=0}
  }
  MaterialExpressionLinearInterpolate_5 {
   class=MaterialExpressionLinearInterpolate
   inputs [
    A {from=MaterialExpressionTextureSampleParameter2D_1, output=G}
    B {from=MaterialExpressionSaturate_2}
    Alpha {from=MaterialExpressionRound_0}
   ]
   properties {constB=0.66}
  }
  MaterialExpressionLinearInterpolate_6 {
   class=MaterialExpressionLinearInterpolate
   inputs [
    A {from=MaterialExpressionLinearInterpolate_5}
    B {from=MaterialExpressionMultiply_13}
    Alpha {from=MaterialExpressionTextureSampleParameter2D_5, output=G}
   ]
  }
  MaterialExpressionLinearInterpolate_7 {
   class=MaterialExpressionLinearInterpolate
   inputs [
    A {from=MaterialExpressionScalarParameter_10}
    B {from=MaterialExpressionScalarParameter_11}
    Alpha {from=MaterialExpressionTextureSampleParameter2D_2, output=R}
   ]
  }
  MaterialExpressionLinearInterpolate_8 {
   class=MaterialExpressionLinearInterpolate
   inputs [
    A {from=MaterialExpressionScalarParameter_12}
    B {from=MaterialExpressionScalarParameter_13}
    Alpha {from=MaterialExpressionTextureSampleParameter2D_7, output=B}
   ]
  }
  MaterialExpressionLinearInterpolate_9 {
   class=MaterialExpressionLinearInterpolate
   inputs [
    A {from=MaterialExpressionLinearInterpolate_10}
    B {from=MaterialExpressionVectorParameter_1, output=RGB}
    Alpha {from=MaterialExpressionCollectionParameter_3}
   ]
  }
  MaterialExpressionMaterialFunctionCall_0 {
   class=MaterialExpressionMaterialFunctionCall
   inputs [
    Rotation Axis {from=MaterialExpressionConstant3Vector_0}
    Rotation Angle {from=MaterialExpressionMultiply_7}
    World Space Vertex Normals {from=""}
    Movement Mask {from=""}
    Normals Pivot {from=""}
   ]
   properties {materialFunction=&FixRotateAboutAxisNormals}
  }
  MaterialExpressionMaterialFunctionCall_1 {
   class=MaterialExpressionMaterialFunctionCall
   inputs [
    BaseNormal {from=MaterialExpressionMaterialFunctionCall_0, output=New Tangent Space Vertex Normal}
    AdditionalNormal {from=MaterialExpressionTextureSampleParameter2D_0, output=RGB}
   ]
   properties {materialFunction=&BlendAngleCorrectedNormals}
  }
  MaterialExpressionMaterialFunctionCall_2 {
   class=MaterialExpressionMaterialFunctionCall
   inputs []
   properties {materialFunction=&ObjectPivotPoint}
  }
  MaterialExpressionMaterialFunctionCall_3 {
   class=MaterialExpressionMaterialFunctionCall
   inputs [
    Material {from=MaterialExpressionMaterialFunctionCall_7, output=Result}
    NewOpacityMask {from=MaterialExpressionMaterialFunctionCall_8, output=OpacityMask}
    Mask {from=""}
   ]
   properties {materialFunction=/Engine/Functions/MaterialLayerFunctions/MatLayerBlend_OverrideOpacityMask}
  }
  MaterialExpressionMaterialFunctionCall_4 {
   class=MaterialExpressionMaterialFunctionCall
   inputs [
    BaseNormal {from=MaterialExpressionTextureSampleParameter2D_0, output=RGB}
    AdditionalNormal {from=MaterialExpressionMaterialFunctionCall_5, output=Result}
   ]
   properties {materialFunction=&BlendAngleCorrectedNormals}
  }
  MaterialExpressionMaterialFunctionCall_5 {
   class=MaterialExpressionMaterialFunctionCall
   inputs [
    Normal {from=MaterialExpressionTextureSampleParameter2D_8, output=RGB}
    Flatness {from=MaterialExpressionScalarParameter_6}
   ]
   properties {materialFunction=&FlattenNormal}
  }
  MaterialExpressionMaterialFunctionCall_6 {
   class=MaterialExpressionMaterialFunctionCall
   inputs [
    AO {from=MaterialExpressionTextureSample_2, output=R}
    In {from=MaterialExpressionSetMaterialAttributes_0}
    Alpha {from=MaterialExpressionTextureSampleParameter2D_5, output=R}
   ]
   properties {materialFunction=&/FlyingOrbCubeofCubes}
  }
  MaterialExpressionMaterialFunctionCall_7 {
   class=MaterialExpressionMaterialFunctionCall
   inputs [
    Alpha {from=MaterialExpressionTextureSampleParameter2D_5, output=B}
    AO {from=MaterialExpressionTextureSample_2, output=R}
    Mat {from=MaterialExpressionMaterialFunctionCall_6, output=Result}
   ]
   properties {materialFunction=&/MF_FlyingOrbWindows}
  }
  MaterialExpressionMaterialFunctionCall_8 {
   class=MaterialExpressionMaterialFunctionCall
   inputs [
    MaskPosition {from=MaterialExpressionComponentMask_0}
    MaskDirection {from=MaterialExpressionConstant3Vector_1}
    MaskRadius {from=MaterialExpressionMultiply_10}
    MaskEdgeGradientSize {from=""}
    MaskEdgeGradientOffset {from=""}
    MaskEdgeGradientPower {from=""}
   ]
   properties {materialFunction=/Game/Effects/Niagara/MasterMaterials/Functions/Opacity/MF_FX_Circular_Portal_Mask}
  }
  MaterialExpressionMultiply_0 {
   class=MaterialExpressionMultiply
   inputs [
    A {from=MaterialExpressionMultiply_16}
    B {from=MaterialExpressionOneMinus_0}
   ]
  }
  MaterialExpressionMultiply_1 {
   class=MaterialExpressionMultiply
   inputs [
    A {from=MaterialExpressionTextureCoordinate_1}
    B {from=MaterialExpressionScalarParameter_7}
   ]
  }
  MaterialExpressionMultiply_10 {
   class=MaterialExpressionMultiply
   inputs [
    A {from=MaterialExpressionAdd_1}
    B {from=MaterialExpressionCollectionParameter_1}
   ]
   properties {constA=5000}
  }
  MaterialExpressionMultiply_11 {
   class=MaterialExpressionMultiply
   inputs [
    A {from=MaterialExpressionTextureSampleParameter2D_1, output=B}
    B {from=MaterialExpressionConstant_1}
   ]
  }
  MaterialExpressionMultiply_12 {
   class=MaterialExpressionMultiply
   inputs [
    A {from=MaterialExpressionTextureSampleParameter2D_6, output=RGB}
    B {from=MaterialExpressionTextureSampleParameter2D_3, output=RGB}
   ]
  }
  MaterialExpressionMultiply_13 {
   class=MaterialExpressionMultiply
   inputs [
    A {from=MaterialExpressionLinearInterpolate_5}
    B {from=MaterialExpressionLinearInterpolate_7}
   ]
  }
  MaterialExpressionMultiply_14 {
   class=MaterialExpressionMultiply
   inputs [
    A {from=MaterialExpressionMultiply_13}
    B {from=""}
   ]
  }
  MaterialExpressionMultiply_15 {
   class=MaterialExpressionMultiply
   inputs [
    A {from=MaterialExpressionLinearInterpolate_8}
    B {from=MaterialExpressionTextureSampleParameter2D_1, output=G}
   ]
  }
  MaterialExpressionMultiply_16 {
   class=MaterialExpressionMultiply
   inputs [
    A {from=MaterialExpressionLinearInterpolate_0}
    B {from=MaterialExpressionAdd_5}
   ]
   properties {constB=0}
  }
  MaterialExpressionMultiply_18 {
   class=MaterialExpressionMultiply
   inputs [
    A {from=MaterialExpressionVectorParameter_0, output=RGB}
    B {from=""}
   ]
   properties {constB=0.05}
  }
  MaterialExpressionMultiply_2 {
   class=MaterialExpressionMultiply
   inputs [
    A {from=MaterialExpressionTime_0}
    B {from=MaterialExpressionScalarParameter_1}
   ]
  }
  MaterialExpressionMultiply_3 {
   class=MaterialExpressionMultiply
   inputs [
    A {from=MaterialExpressionMultiply_2}
    B {from=""}
   ]
   properties {constB=3.14}
  }
  MaterialExpressionMultiply_4 {
   class=MaterialExpressionMultiply
   inputs [
    A {from=MaterialExpressionSine_0}
    B {from=MaterialExpressionSine_1}
   ]
  }
  MaterialExpressionMultiply_5 {
   class=MaterialExpressionMultiply
   inputs [
    A {from=MaterialExpressionScalarParameter_2}
    B {from=MaterialExpressionAdd_0}
   ]
  }
  MaterialExpressionMultiply_6 {
   class=MaterialExpressionMultiply
   inputs [
    A {from=MaterialExpressionScalarParameter_3}
    B {from=MaterialExpressionSine_0}
   ]
  }
  MaterialExpressionMultiply_7 {
   class=MaterialExpressionMultiply
   inputs [
    A {from=MaterialExpressionMultiply_4}
    B {from=MaterialExpressionScalarParameter_4}
   ]
  }
  MaterialExpressionMultiply_8 {
   class=MaterialExpressionMultiply
   inputs [
    A {from=MaterialExpressionTextureCoordinate_0}
    B {from=MaterialExpressionScalarParameter_5}
   ]
  }
  MaterialExpressionMultiply_9 {
   class=MaterialExpressionMultiply
   inputs [
    A {from=MaterialExpressionConvert_0}
    B {from=MaterialExpressionCollectionParameter_0}
   ]
  }
  MaterialExpressionOneMinus_0 {class=MaterialExpressionOneMinus, inputs [None {from=MaterialExpressionFresnel_0}]}
  MaterialExpressionPower_0 {
   class=MaterialExpressionPower
   inputs [
    Base {from=MaterialExpressionSaturate_3}
    Exp {from=""}
   ]
   properties {constExponent=0.6}
  }
  MaterialExpressionRotateAboutAxis_0 {
   class=MaterialExpressionRotateAboutAxis
   inputs [
    NormalizedRotationAxis {from=MaterialExpressionConstant3Vector_0, output=""}
    RotationAngle {from=MaterialExpressionMultiply_7, output=""}
    PivotPoint {from=MaterialExpressionMaterialFunctionCall_2, output=Object Pivot Location}
    Position {from=MaterialExpressionActorPositionWS_0, output=""}
   ]
   properties {}
  }
  MaterialExpressionRound_0 {class=MaterialExpressionRound, inputs [None {from=MaterialExpressionMultiply_11, output=""}], properties {}}
  MaterialExpressionSaturate_0 {class=MaterialExpressionSaturate, inputs [None {from=MaterialExpressionSubtract_0}]}
  MaterialExpressionSaturate_1 {class=MaterialExpressionSaturate, inputs [None {from=MaterialExpressionLinearInterpolate_6}]}
  MaterialExpressionSaturate_2 {class=MaterialExpressionSaturate, inputs [None {from=MaterialExpressionMultiply_15}]}
  MaterialExpressionSaturate_3 {class=MaterialExpressionSaturate, inputs [None {from=MaterialExpressionDivide_0}]}
  MaterialExpressionScalarParameter_0 {class=MaterialExpressionScalarParameter, properties {defaultValue=0.3983, parameterName=EmissiveFresnel}}
  MaterialExpressionScalarParameter_1 {class=MaterialExpressionScalarParameter, properties {defaultValue=0.018, parameterName=WobbleSpeed}}
  MaterialExpressionScalarParameter_10 {class=MaterialExpressionScalarParameter, properties {defaultValue=0.5371, group=Panels, parameterName=Panel Roughness Min}}
  MaterialExpressionScalarParameter_11 {class=MaterialExpressionScalarParameter, properties {defaultValue=0.9612, group=Panels, parameterName=Panel Roughness Max}}
  MaterialExpressionScalarParameter_12 {class=MaterialExpressionScalarParameter, properties {defaultValue=0.6498, parameterName=Metal Detail Roughness Min}}
  MaterialExpressionScalarParameter_13 {class=MaterialExpressionScalarParameter, properties {defaultValue=1.3832, parameterName=Metal Detail Roughness Max}}
  MaterialExpressionScalarParameter_2 {class=MaterialExpressionScalarParameter, properties {defaultValue=170, parameterName=WobbleVertIntensity}}
  MaterialExpressionScalarParameter_3 {class=MaterialExpressionScalarParameter, properties {defaultValue=60, parameterName=WobblhorizontalIntensity}}
  MaterialExpressionScalarParameter_4 {class=MaterialExpressionScalarParameter, properties {defaultValue=0.0138, parameterName=RotationAmount}}
  MaterialExpressionScalarParameter_5 {class=MaterialExpressionScalarParameter, properties {defaultValue=8, group=Panels, parameterName=PanelDetailTiling}}
  MaterialExpressionScalarParameter_6 {class=MaterialExpressionScalarParameter, properties {defaultValue=0.4, parameterName=Panel Flatness}}
  MaterialExpressionScalarParameter_7 {class=MaterialExpressionScalarParameter, properties {defaultValue=40, group=Metal, parameterName=Metal Detail Tiling}}
  MaterialExpressionScalarParameter_8 {class=MaterialExpressionScalarParameter, properties {defaultValue=5000, parameterName=PortalMaxSize}}
  MaterialExpressionScalarParameter_9 {class=MaterialExpressionScalarParameter, properties {defaultValue=2.5, parameterName=HostileModePulseTime}}
  MaterialExpressionSetMaterialAttributes_0 {
   class=MaterialExpressionSetMaterialAttributes
   inputs [
    MaterialAttributes {from="", output=""}
    Base Color {from=MaterialExpressionLinearInterpolate_1, output=""}
    Metallic {from=MaterialExpressionTextureSampleParameter2D_1, output=B}
    Specular {from=MaterialExpressionTextureSampleParameter2D_1, output=B}
    Roughness {from=MaterialExpressionSaturate_1, output=""}
    Anisotropy {from="", output=""}
    Emissive Color {from=MaterialExpressionLinearInterpolate_4, output=""}
    Opacity {from="", output=""}
    Opacity Mask {from="", output=""}
    Normal {from=MaterialExpressionLinearInterpolate_2, output=""}
    Tangent {from="", output=""}
    World Position Offset {from=MaterialExpressionMultiply_9, output=""}
   ]
   properties {attributeSetTypes [
     #1
     #2
     #3
     #4
     #5
     #6
     #7
     #8
     #9
     #10
     #11
    ]}
  }
  MaterialExpressionSine_0 {class=MaterialExpressionSine, inputs [None {from=MaterialExpressionMultiply_2}]}
  MaterialExpressionSine_1 {class=MaterialExpressionSine, inputs [None {from=MaterialExpressionMultiply_3}]}
  MaterialExpressionSine_2 {class=MaterialExpressionSine, inputs [None {from=MaterialExpressionPower_0}]}
  MaterialExpressionSubtract_0 {
   class=MaterialExpressionSubtract
   inputs [
    A {from=MaterialExpressionTextureSampleParameter2D_4, output=R}
    B {from=MaterialExpressionTextureSampleParameter2D_5, output=R}
   ]
  }
  MaterialExpressionTextureCoordinate_0 {class=MaterialExpressionTextureCoordinate}
  MaterialExpressionTextureCoordinate_1 {class=MaterialExpressionTextureCoordinate}
  MaterialExpressionTextureCoordinate_2 {class=MaterialExpressionTextureCoordinate}
  MaterialExpressionTextureSampleParameter2D_0 {
   class=MaterialExpressionTextureSampleParameter2D
   inputs [
    UVs {from="", output=""}
    Apply View MipBias {from="", output=""}
   ]
   properties {parameterName=Normal, samplerType=SAMPLERTYPE_Normal, texture=/DragonCartNarrative/Assets/FlyingOrb/Textures/T_FlyingOrb_FlyingOrb_B_N}
  }
  MaterialExpressionTextureSampleParameter2D_1 {
   class=MaterialExpressionTextureSampleParameter2D
   inputs [
    UVs {from="", output=""}
    Apply View MipBias {from="", output=""}
   ]
   properties {parameterName=SRM, samplerType=SAMPLERTYPE_Masks, texture=/DragonCartNarrative/Assets/FlyingOrb/Textures/T_FlyingOrb_FlyingOrb_B_S}
  }
  MaterialExpressionTextureSampleParameter2D_2 {
   class=MaterialExpressionTextureSampleParameter2D
   inputs [
    UVs {from=MaterialExpressionMultiply_8, output=""}
    Apply View MipBias {from="", output=""}
   ]
   properties {group=Panels, parameterName=Panel Roughness, texture=&/FlyingOrb_Panels_Pattern_Roughness}
  }
  MaterialExpressionTextureSampleParameter2D_3 {
   class=MaterialExpressionTextureSampleParameter2D
   inputs [
    UVs {from="", output=""}
    Apply View MipBias {from="", output=""}
   ]
   properties {parameterName=DIFF, texture=/DragonCartNarrative/Assets/FlyingOrb/Textures/T_FlyingOrb_FlyingOrb_B_D}
  }
  MaterialExpressionTextureSampleParameter2D_4 {
   class=MaterialExpressionTextureSampleParameter2D
   inputs [
    UVs {from="", output=""}
    Apply View MipBias {from="", output=""}
   ]
   properties {parameterName=Emissive, texture=/DragonCartNarrative/Assets/FlyingOrb/Textures/T_FlyingOrb_FlyingOrb_B_E}
  }
  MaterialExpressionTextureSampleParameter2D_5 {
   class=MaterialExpressionTextureSampleParameter2D
   inputs [
    UVs {from="", output=""}
    Apply View MipBias {from="", output=""}
   ]
   properties {parameterName=FX, texture=/DragonCartNarrative/Assets/FlyingOrb/Textures/T_FlyingOrb_FlyingOrb_B_FX}
  }
  MaterialExpressionTextureSampleParameter2D_6 {
   class=MaterialExpressionTextureSampleParameter2D
   inputs [
    UVs {from=MaterialExpressionMultiply_8, output=""}
    Apply View MipBias {from="", output=""}
   ]
   properties {group=Panels, parameterName=Panel diffuse, texture=&/FlyingOrbPanels_D}
  }
  MaterialExpressionTextureSampleParameter2D_7 {
   class=MaterialExpressionTextureSampleParameter2D
   inputs [
    UVs {from=MaterialExpressionMultiply_1, output=""}
    Apply View MipBias {from="", output=""}
   ]
   properties {group=Metal, parameterName="Metal Detail (B)", texture=&/LightsFlyingOrb_M}
  }
  MaterialExpressionTextureSampleParameter2D_8 {
   class=MaterialExpressionTextureSampleParameter2D
   inputs [
    UVs {from=MaterialExpressionMultiply_8, output=""}
    Apply View MipBias {from="", output=""}
   ]
   properties {parameterName=Panel Normal, samplerType=SAMPLERTYPE_Normal, texture=&/FlyingOrbPanels_M}
  }
  MaterialExpressionTextureSample_2 {
   class=MaterialExpressionTextureSample
   inputs [
    UVs {from=""}
    Tex {from=""}
    Apply View MipBias {from=""}
   ]
   properties {texture=/DragonCartNarrative/Assets/FlyingOrb/Textures/T_FlyingOrb_FlyingOrb_B_M}
  }
  MaterialExpressionTime_0 {class=MaterialExpressionTime}
  MaterialExpressionTime_1 {class=MaterialExpressionTime}
  MaterialExpressionVectorParameter_0 {class=MaterialExpressionVectorParameter, properties {parameterName=EmissiveBasic, defaultValue=<rgba 12.1875 3.5183 0 1>}}
  MaterialExpressionVectorParameter_1 {
   class=MaterialExpressionVectorParameter
   properties {group=Attack Mode, parameterName=AttackMode_EmissiveBasicColor, defaultValue=<rgba 10 0.2604 0.2604 1>}
  }
  MaterialExpressionVectorParameter_2 {
   class=MaterialExpressionVectorParameter
   properties {group=Attack Mode, parameterName=HostileMode_EmissiveBasicColor, defaultValue=<rgba 25 0 0 1>}
  }
 }
 properties {
  bAutomaticallySetUsageInEditor=false
  bUseMaterialAttributes=true
  bUsedWithInstancedStaticMeshes=true
  bUsedWithMeshDeformer=true
  bUsedWithSkeletalMesh=true
  twoSided=true
  usedShadingModels=Default Lit
 }
}
