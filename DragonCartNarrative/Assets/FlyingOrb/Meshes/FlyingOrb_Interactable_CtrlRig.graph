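# Control Rig for the FlyingOrb interactable, animating its body, bay doors and cannon with arms.
# Construction procedurally spawns controls from the bone list by index and name, plus pivot nulls and animation channels.
# Forward solve aims a center null between top and bottom pivot controls to move main unless Break Pivot is set, copies controls to bones and toggles pivot visibility.
# Backward solve restores controls from bones, reprojecting pivots from main.
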ControlRigBlueprint /DragonCartNarrative/Assets/FlyingOrb/Meshes/FlyingOrb_Interactable_CtrlRig
graphs [
 New Function_ContainedGraph {nodes [
   n1 = RigVMFunctionEntryNode()
   RigVMFunction_Sequence = &RigVMFunction_Sequence(ExecuteContext=@n1.ExecuteContext)
   RigVMVariableNode(
    Value=/Script/ControlRig.RigUnit_CollectionGetAll(TypeToSearch=Bone) {is=struct RigVMUnitNode}.Items
    ExecuteContext=@RigVMFunction_Sequence.A
   ) {variable=Skeleton Array}
   RigVMFunctionReturnNode(ExecuteContext=@RigVMFunction_Sequence.B)
  ]}
 New Function_ContainedGraph {nodes [
   n1 = RigVMFunctionEntryNode()
   RigVMFunction_Sequence = &RigVMFunction_Sequence(ExecuteContext=@n1.ExecuteContext)
   RigVMDispatchNode(ExecuteContext=@RigVMFunction_Sequence.A, Array=RigVMVariableNode() {variable=Skeleton Array}.Value) {template="DISPATCH_RigVMDispatch_ArrayReset(io Array)"}
   > RigVMDispatchNode(Array=RigVMVariableNode() {variable=Control Array}.Value) {template="DISPATCH_RigVMDispatch_ArrayReset(io Array)"}
   > RigVMDispatchNode(Array=RigVMVariableNode() {variable=Anim Channel Array}.Value) {template="DISPATCH_RigVMDispatch_ArrayReset(io Array)"}
   RigVMFunctionReturnNode(ExecuteContext=@RigVMFunction_Sequence.B)
  ]}
 New Function_ContainedGraph {nodes [
   RigVMFunctionEntryNode()
   > RigVMFunction_Sequence = RigVMAggregateNode()
   For_Each = RigVMDispatchNode(ExecuteContext=@RigVMFunction_Sequence.A, Array=RigVMVariableNode() {variable=Skeleton Array}.Value) {template="DISPATCH_RigVMDispatch_ArrayIterator(in Array,out Element,out Index,out Count,out Ratio)"}
   > RigVMVariableNode(Value=@For_Each.Element) {variable=Element}
   > RigVMVariableNode(Value=@For_Each.Index) {variable=Index}
   > RigVMFunction_Sequence_1 = RigVMAggregateNode()
   LessEqual = /Script/RigVM.RigVMFunction_MathIntLessEqual(B=1, A=RigVMVariableNode() {variable=Index}.Value) {is=struct RigVMUnitNode, template="LessEqual::Execute(in A,in B,out Result)"}
   RigVMFunction_ControlFlowBranch = &RigVMFunction_ControlFlowBranch(Condition=@LessEqual.Result, ExecuteContext=@RigVMFunction_Sequence_1.A)
   n1 = RigVMVariableNode() {variable=Element}
   n2 = RigVMDispatchNode(
    False=(bVisible=True,Name="Circle_Thin",Color=(R=0,G=0.9434,B=1,A=1),Transform=(Rotation=(X=0,Y=0,Z=0,W=1),Translation=(X=0,Y=0,Z=0),Scale3D=(X=1400,Y=1400,Z=1400)))
    True=(bVisible=True,Name="RoundedSquare_Thin",Color=(R=0,G=1,B=0,A=1),Transform=(Rotation=(X=0,Y=0,Z=0,W=1),Translation=(X=0,Y=0,Z=0),Scale3D=(X=1600,Y=1600,Z=1600)))
    Condition=RigVMDispatchNode(B=0, A=RigVMVariableNode() {variable=Index}.Value) {template="DISPATCH_RigVMDispatch_CoreEquals(in A,in B,out Result)"}.Result
   ) {template="DISPATCH_RigVMDispatch_If(in Condition,in True,in False,out Result)"}
   n3 = /Script/ControlRig.RigUnit_HierarchyAddControlTransform(
    OffsetSpace=GlobalSpace
    OffsetTransform=(Rotation=(X=0,Y=0,Z=0,W=1),Translation=(X=0,Y=0,Z=0),Scale3D=(X=1,Y=1,Z=1))
    Settings="(InitialSpace=LocalSpace,bUsePreferredRotationOrder=False,PreferredRotationOrder=YZX,Limits=(LimitTranslationX=(bMinimum=False,bMaximum=False),LimitTranslationY=(bMinimum=False,bMaximum=False),LimitTranslationZ=(bMinimum=False,bMaximum=False),LimitPitch=(b ...(truncated)"
    InitialValue=&RigUnit_GetTransform(bInitial=False, Item=@n1.Value).Transform
    Settings.Shape=@n2.Result
    ExecutePin=@RigVMFunction_ControlFlowBranch.True
    Parent=RigVMVariableNode() {variable=Running Parent}.Value
    Name=@n1.Value.Name
   ) {
    is=struct RigVMUnitNode
    template="SpawnControl::Execute(in InitialValue,in Settings,in OffsetTransform,in OffsetSpace,in Parent,in Name,out Item)"
   }
   > RigVMVariableNode(Value=@n3.Item) {variable=Running Parent}
   > Add = RigVMDispatchNode(Array=RigVMVariableNode() {variable=Control Array}.Value, Element=@n3.Item) {template="DISPATCH_RigVMDispatch_ArrayAdd(io Array,in Element,out Index)"}
   n4 = &RigVMFunction_ControlFlowBranch(
    Condition=/Script/RigVM.RigVMFunction_Contains(Search=cannon, Name=RigVMVariableNode() {variable=Element}.Value.Name) {is=struct RigVMUnitNode, template="Contains::Execute(in Name,in Search,out Result)"}.Result
    ExecuteContext=@RigVMFunction_Sequence_1.D
   )
   RigVMFunction_Sequence_2 = RigVMAggregateNode(ExecuteContext=@n4.True)
   n5 = &RigVMFunction_ControlFlowBranch(
    Condition=&RigVMFunction_Contains(Search=main, Name=RigVMVariableNode() {variable=Element}.Value.Name) {template="Contains::Execute(in Name,in Search,out Result)"}.Result
    ExecuteContext=@RigVMFunction_Sequence_2.B
   )
   n6 = RigVMVariableNode() {variable=Element}
   n7 = &RigUnit_HierarchyAddControlTransform(
    OffsetSpace=GlobalSpace
    OffsetTransform=(Rotation=(X=0,Y=0,Z=0,W=1),Translation=(X=0,Y=0,Z=0),Scale3D=(X=1,Y=1,Z=1))
    Settings="(InitialSpace=LocalSpace,bUsePreferredRotationOrder=False,PreferredRotationOrder=YZX,Limits=(LimitTranslationX=(bMinimum=False,bMaximum=False),LimitTranslationY=(bMinimum=False,bMaximum=False),LimitTranslationZ=(bMinimum=False,bMaximum=False),LimitPitch=(b ...(truncated)"
    InitialValue=&RigUnit_GetTransform(bInitial=False, Item=@n6.Value).Transform
    ExecutePin=@n5.True
    Name=@n6.Value.Name
    Parent=RigVMVariableNode() {variable=Running Parent}.Value
   ) {template="SpawnControl::Execute(in InitialValue,in Settings,in OffsetTransform,in OffsetSpace,in Parent,in Name,out Item)"}
   > RigVMVariableNode(Value=@n7.Item) {variable=Running Parent}
   > RigVMDispatchNode(Element=@n7.Item, Array=RigVMVariableNode() {variable=Control Array}.Value) {template="DISPATCH_RigVMDispatch_ArrayAdd(io Array,in Element,out Index)"}
   n8 = &RigVMFunction_ControlFlowBranch(
    Condition=&RigVMFunction_Contains(Search=apparatus, Name=RigVMVariableNode() {variable=Element}.Value.Name) {template="Contains::Execute(in Name,in Search,out Result)"}.Result
    ExecuteContext=@RigVMFunction_Sequence_2.A
   )
   n9 = RigVMVariableNode() {variable=Element}
   n10 = &RigUnit_HierarchyAddControlTransform(
    OffsetSpace=GlobalSpace
    OffsetTransform=(Rotation=(X=0,Y=0,Z=0,W=1),Translation=(X=0,Y=0,Z=0),Scale3D=(X=1,Y=1,Z=1))
    Settings="(InitialSpace=LocalSpace,bUsePreferredRotationOrder=False,PreferredRotationOrder=YZX,Limits=(LimitTranslationX=(bMinimum=False,bMaximum=False),LimitTranslationY=(bMinimum=False,bMaximum=False),LimitTranslationZ=(bMinimum=False,bMaximum=False),LimitPitch=(b ...(truncated)"
    InitialValue=&RigUnit_GetTransform(bInitial=False, Item=@n9.Value).Transform
    ExecutePin=@n8.True
    Name=@n9.Value.Name
    Parent=RigVMVariableNode() {variable=Running Parent}.Value
   ) {template="SpawnControl::Execute(in InitialValue,in Settings,in OffsetTransform,in OffsetSpace,in Parent,in Name,out Item)"}
   > RigVMVariableNode(Value=@n10.Item) {variable=Running Parent}
   > RigVMDispatchNode(Element=@n10.Item, Array=RigVMVariableNode() {variable=Control Array}.Value) {template="DISPATCH_RigVMDispatch_ArrayAdd(io Array,in Element,out Index)"}
   n11 = &RigVMFunction_ControlFlowBranch(
    Condition=&RigVMFunction_Contains(Search=arm, Name=RigVMVariableNode() {variable=Element}.Value.Name) {template="Contains::Execute(in Name,in Search,out Result)"}.Result
    ExecuteContext=@RigVMFunction_Sequence_2.C
   )
   n12 = RigVMVariableNode() {variable=Element}
   n13 = &RigUnit_HierarchyAddControlTransform(
    OffsetSpace=GlobalSpace
    OffsetTransform=(Rotation=(X=0,Y=0,Z=0,W=1),Translation=(X=0,Y=0,Z=0),Scale3D=(X=1,Y=1,Z=1))
    Settings="(InitialSpace=LocalSpace,bUsePreferredRotationOrder=False,PreferredRotationOrder=YZX,Limits=(LimitTranslationX=(bMinimum=False,bMaximum=False),LimitTranslationY=(bMinimum=False,bMaximum=False),LimitTranslationZ=(bMinimum=False,bMaximum=False),LimitPitch=(b ...(truncated)"
    InitialValue=&RigUnit_GetTransform(bInitial=False, Item=@n12.Value).Transform
    ExecutePin=@n11.True
    Name=@n12.Value.Name
    Parent=RigVMVariableNode() {variable=Running Parent}.Value
   ) {template="SpawnControl::Execute(in InitialValue,in Settings,in OffsetTransform,in OffsetSpace,in Parent,in Name,out Item)"}
   > RigVMDispatchNode(Element=@n13.Item, Array=RigVMVariableNode() {variable=Control Array}.Value) {template="DISPATCH_RigVMDispatch_ArrayAdd(io Array,in Element,out Index)"}
   n14 = &RigVMFunction_ControlFlowBranch(
    Condition=RigVMDispatchNode(B=2, A=RigVMVariableNode() {variable=Index}.Value) {template="DISPATCH_RigVMDispatch_CoreEquals(in A,in B,out Result)"}.Result
    ExecuteContext=@RigVMFunction_Sequence_1.B
   )
   n15 = RigVMVariableNode() {variable=Element}
   n16 = &RigUnit_GetTransform(bInitial=False, Item=@n15.Value)
   n17 = &RigUnit_HierarchyAddNull(Name=MainNull, Space=GlobalSpace, Transform=@n16.Transform, ExecutePin=@n14.True, Parent=RigVMVariableNode() {variable=Running Parent}.Value)
   > SpawnControl = &RigUnit_HierarchyAddControlTransform(
    InitialValue=(Rotation=(X=0,Y=0,Z=0,W=1),Translation=(X=0,Y=0,Z=0),Scale3D=(X=1,Y=1,Z=1))
    OffsetSpace=GlobalSpace
    Settings="(InitialSpace=LocalSpace,bUsePreferredRotationOrder=False,PreferredRotationOrder=YZX,Limits=(LimitTranslationX=(bMinimum=False,bMaximum=False),LimitTranslationY=(bMinimum=False,bMaximum=False),LimitTranslationZ=(bMinimum=False,bMaximum=False),LimitPitch=(b ...(truncated)"
    Parent=@n17.Item
    OffsetTransform=@n16.Transform
    Name=@n15.Value.Name
   ) {template="SpawnControl::Execute(in InitialValue,in Settings,in OffsetTransform,in OffsetSpace,in Parent,in Name,out Item)"}
   > RigVMVariableNode(Value=@SpawnControl.Item) {variable=Running Parent}
   > RigVMDispatchNode(Array=RigVMVariableNode() {variable=Control Array}.Value, Element=@SpawnControl.Item) {template="DISPATCH_RigVMDispatch_ArrayAdd(io Array,in Element,out Index)"}
   n18 = RigVMVariableNode() {variable=Element}
   Contains = &RigVMFunction_Contains(Search=bay_door, Name=@n18.Value.Name) {template="Contains::Execute(in Name,in Search,out Result)"}
   n19 = &RigVMFunction_ControlFlowBranch(Condition=@Contains.Result, ExecuteContext=@RigVMFunction_Sequence_1.C)
   Make_Array = RigVMDispatchNode(Values=((R=0,G=0,B=1,A=0),((R=1,G=1,B=1,A=1),R=0,G=0,B=1,A=0),((R=1,G=1,B=1,A=1),R=1,G=0,B=0,A=0),((R=1,G=1,B=1,A=1),R=1,G=0,B=0,A=0),((R=1,G=1,B=1,A=1),R=0,G=1,B=0,A=0),((R=1,G=1,B=1,A=1),R=0,G=1,B=0,A=0))) {template="DISPATCH_RigVMDispatch_ArrayMake(in Values,out Array)"}
   n20 = &RigVMFunction_Contains(Search=inner, Name=@n18.Value.Name) {template="Contains::Execute(in Name,in Search,out Result)"}
   n21 = RigVMDispatchNode(
    False=(Rotation=(X=0,Y=0,Z=1,W=0),Translation=(X=0,Y=1050,Z=-6000),Scale3D=(X=200,Y=100,Z=200))
    True=(Rotation=(X=0,Y=0,Z=1,W=0),Translation=(X=0,Y=2000,Z=-6000),Scale3D=(X=280,Y=140,Z=240))
    False.Translation.Z=&RigVMFunction_MathDoubleAdd(
     B=-5800
     A=RigVMDispatchNode(False=100, True=0, Condition=@n20.Result) {template="DISPATCH_RigVMDispatch_If(in Condition,in True,in False,out Result)"}.Result
    ).Result
    Condition=&RigVMFunction_Contains(Search=outer, Name=@n18.Value.Name) {template="Contains::Execute(in Name,in Search,out Result)"}.Result
   ) {template="DISPATCH_RigVMDispatch_If(in Condition,in True,in False,out Result)"}
   n22 = &RigUnit_HierarchyAddControlTransform(
    OffsetSpace=GlobalSpace
    OffsetTransform=(Rotation=(X=0,Y=0,Z=0,W=1),Translation=(X=0,Y=0,Z=0),Scale3D=(X=1,Y=1,Z=1))
    Parent=(Type=Control,Name="main")
    Settings="(InitialSpace=LocalSpace,bUsePreferredRotationOrder=False,PreferredRotationOrder=YZX,Limits=(LimitTranslationX=(bMinimum=False,bMaximum=False),LimitTranslationY=(bMinimum=False,bMaximum=False),LimitTranslationZ=(bMinimum=False,bMaximum=False),LimitPitch=(b ...(truncated)"
    Settings.Shape.Color=RigVMDispatchNode(Array=@Make_Array.Array, Index=RigVMVariableNode() {variable=Count}.Value) {template="DISPATCH_RigVMDispatch_ArrayGetAtIndex(in Array,in Index,out Element)"}.Element
    InitialValue=&RigUnit_GetTransform(bInitial=False, Item=@n18.Value).Transform
    Settings.Shape.Transform=@n21.Result
    ExecutePin=@n19.True
    Name=@n18.Value.Name
   ) {template="SpawnControl::Execute(in InitialValue,in Settings,in OffsetTransform,in OffsetSpace,in Parent,in Name,out Item)"}
   > RigVMVariableNode(Value=/Script/RigVM.RigVMFunction_MathIntAdd(B=1, A=RigVMVariableNode() {variable=Count}.Value) {is=struct RigVMUnitNode, template="Add::Execute(in A,in B,out Result)"}.Result) {variable=Count}
   > RigVMDispatchNode(Array=RigVMVariableNode() {variable=Control Array}.Value, Element=@n22.Item) {template="DISPATCH_RigVMDispatch_ArrayAdd(io Array,in Element,out Index)"}
   n23 = &RigVMFunction_Sequence(ExecuteContext=@RigVMFunction_Sequence.B)
   n24 = &RigUnit_HierarchyAddControlTransform(
    InitialValue=(Rotation=(X=0,Y=0,Z=0,W=1),Translation=(X=0,Y=0,Z=0),Scale3D=(X=1,Y=1,Z=1))
    Name=pivot_top
    OffsetSpace=GlobalSpace
    OffsetTransform=(Rotation=(X=0,Y=0,Z=0,W=1),Translation=(X=0,Y=0,Z=5200),Scale3D=(X=1,Y=1,Z=1))
    Parent=(Type=Control,Name="offset")
    Settings="(InitialSpace=LocalSpace,bUsePreferredRotationOrder=False,PreferredRotationOrder=YZX,Limits=(LimitTranslationX=(bMinimum=False,bMaximum=False),LimitTranslationY=(bMinimum=False,bMaximum=False),LimitTranslationZ=(bMinimum=False,bMaximum=False),LimitPitch=(b ...(truncated)"
    ExecutePin=@n23.A
   ) {template="SpawnControl::Execute(in InitialValue,in Settings,in OffsetTransform,in OffsetSpace,in Parent,in Name,out Item)"}
   > n25 = &RigUnit_HierarchyAddControlTransform(
    InitialValue=(Rotation=(X=0,Y=0,Z=0,W=1),Translation=(X=0,Y=0,Z=0),Scale3D=(X=1,Y=1,Z=1))
    Name=pivot_bottom
    OffsetSpace=GlobalSpace
    OffsetTransform=(Rotation=(X=0,Y=0,Z=0,W=1),Translation=(X=0,Y=0,Z=-5000),Scale3D=(X=1,Y=1,Z=1))
    Parent=(Type=Control,Name="offset")
    Settings="(InitialSpace=LocalSpace,bUsePreferredRotationOrder=False,PreferredRotationOrder=YZX,Limits=(LimitTranslationX=(bMinimum=False,bMaximum=False),LimitTranslationY=(bMinimum=False,bMaximum=False),LimitTranslationZ=(bMinimum=False,bMaximum=False),LimitPitch=(b ...(truncated)"
   ) {template="SpawnControl::Execute(in InitialValue,in Settings,in OffsetTransform,in OffsetSpace,in Parent,in Name,out Item)"}
   n26 = &RigUnit_GetTransform(Item=(Type=Null,Name="MainNull"), bInitial=False)
   n27 = &RigUnit_HierarchyAddNull(
    Name=pivot_center
    Parent=(Type=Control,Name="offset")
    Space=LocalSpace
    Transform=(Rotation=(X=0,Y=0.7071,Z=0,W=0.7071),Translation=(X=0,Y=0,Z=0),Scale3D=(X=1,Y=1,Z=1))
    Transform.Translation=@n26.Transform.Translation
    ExecutePin=@n23.B
   )
   > &RigUnit_HierarchyAddNull(
    Name=CenterNull
    Space=LocalSpace
    Transform=(Rotation=(X=0,Y=0.7071,Z=0,W=0.7071),Translation=(X=0,Y=0,Z=0),Scale3D=(X=1,Y=1,Z=1))
    Parent=@n27.Item
   )
   RigVMFunctionReturnNode(ExecuteContext=@RigVMFunction_Sequence.C)
   &RigUnit_HierarchyAddNull(
    ExecutePin=()
    Name=pivot_top
    Space=LocalSpace
    Transform=(Rotation=(X=0,Y=0,Z=0,W=1),Translation=(X=0,Y=0,Z=0),Scale3D=(X=1,Y=1,Z=1))
    Parent=@n24.Item
   )
   &RigUnit_HierarchyAddNull(
    ExecutePin=()
    Name=pivot_bottom
    Space=LocalSpace
    Transform=(Rotation=(X=0,Y=0,Z=0,W=1),Translation=(X=0,Y=0,Z=0),Scale3D=(X=1,Y=1,Z=1))
    Parent=@n25.Item
   )
   RigVMFunction_MathBoolNot = &RigVMFunction_MathBoolNot(Value=False)
   &RigUnit_HierarchyAddControlTransform(
    ExecutePin=()
    InitialValue=(Rotation=(X=0,Y=0,Z=0,W=1),Translation=(X=0,Y=0,Z=0),Scale3D=(X=1,Y=1,Z=1))
    Name=pivot_center
    OffsetSpace=GlobalSpace
    OffsetTransform=(Rotation=(X=0.5,Y=-0.5,Z=0.5,W=0.5),Translation=(X=-0.0002,Y=-0.0019,Z=-198.3217),Scale3D=(X=1,Y=1,Z=1))
    Parent=(Type=Control,Name="offset")
    Settings="(InitialSpace=LocalSpace,bUsePreferredRotationOrder=False,PreferredRotationOrder=YZX,Limits=(LimitTranslationX=(bMinimum=False,bMaximum=False),LimitTranslationY=(bMinimum=False,bMaximum=False),LimitTranslationZ=(bMinimum=False,bMaximum=False),LimitPitch=(b ...(truncated)"
    OffsetTransform.Translation=@n26.Transform.Translation
   ) {template="SpawnControl::Execute(in InitialValue,in Settings,in OffsetTransform,in OffsetSpace,in Parent,in Name,out Item)"}
  ]}
 New Function_ContainedGraph {nodes [
   RigVMFunctionEntryNode()
   > RigVMFunction_Sequence = RigVMAggregateNode()
   SpawnAnimationChannel = /Script/ControlRig.RigUnit_HierarchyAddAnimationChannelBool(
    InitialValue=False
    LimitsEnabled=()
    MaximumValue=True
    MinimumValue=False
    Name=Break Pivot
    Parent=(Type=Control,Name="root")
    ExecutePin=@RigVMFunction_Sequence.A
   ) {
    is=struct RigVMUnitNode
    template="SpawnAnimationChannel::Execute(in InitialValue,in MinimumValue,in MaximumValue,in LimitsEnabled,in Parent,in Name,out Item)"
   }
   > Add = RigVMDispatchNode(Element=@SpawnAnimationChannel.Item, Array=RigVMVariableNode() {variable=Anim Channel Array}.Value) {template="DISPATCH_RigVMDispatch_ArrayAdd(io Array,in Element,out Index)"}
   > n1 = &RigUnit_HierarchyAddAnimationChannelBool(InitialValue=False, LimitsEnabled=(), MaximumValue=True, MinimumValue=False, Name=Hide Pivot Controls, Parent=(Type=Control,Name="root")) {template="SpawnAnimationChannel::Execute(in InitialValue,in MinimumValue,in MaximumValue,in LimitsEnabled,in Parent,in Name,out Item)"}
   > RigVMDispatchNode(Element=@n1.Item, Array=RigVMVariableNode() {variable=Anim Channel Array}.Value) {template="DISPATCH_RigVMDispatch_ArrayAdd(io Array,in Element,out Index)"}
   Clone = RigVMDispatchNode(Array=RigVMVariableNode() {variable=Control Array}.Value) {template="DISPATCH_RigVMDispatch_ArrayClone(in Array,out Clone)"}
   n2 = RigVMDispatchNode(Index=0, Array=@Clone.Clone, ExecuteContext=@RigVMFunction_Sequence.B) {template="DISPATCH_RigVMDispatch_ArrayRemove(io Array,in Index)"}
   > RigVMVariableNode(Value=@n2.Array) {variable=Modified Control List}
   > For_Each = RigVMDispatchNode(Array=RigVMVariableNode() {variable=Anim Channel Array}.Value) {template="DISPATCH_RigVMDispatch_ArrayIterator(in Array,out Element,out Index,out Count,out Ratio)"}
   > /Script/ControlRig.RigUnit_SetChannelHosts(Channel=@For_Each.Element, Hosts=RigVMVariableNode() {variable=Modified Control List}.Value) {is=struct RigVMUnitNode}
   RigVMFunctionReturnNode(ExecuteContext=@RigVMFunction_Sequence.C)
  ]}
 New Function_ContainedGraph {nodes [
   RigVMFunctionEntryNode()
   > RigVMFunction_Sequence = RigVMAggregateNode()
   For_Each = RigVMDispatchNode(ExecuteContext=@RigVMFunction_Sequence.A, Array=RigVMVariableNode() {variable=Skeleton Array}.Value) {template="DISPATCH_RigVMDispatch_ArrayIterator(in Array,out Element,out Index,out Count,out Ratio)"}
   n1 = &RigVMFunction_ControlFlowBranch(
    Condition=RigVMDispatchNode(B=2, A=@For_Each.Index) {template="DISPATCH_RigVMDispatch_CoreEquals(in A,in B,out Result)"}.Result
    ExecuteContext=@For_Each.ExecuteContext
   )
   /Script/ControlRig.RigUnit_PositionConstraintLocalSpaceOffset(
    Child=(Type=Null,Name="pivot_center")
    Filter=(bX=True,bY=True,bZ=True)
    Parents=((Item=(Type=Control,Name="pivot_top"),Weight=1),(Item=(Type=Control,Name="pivot_bottom"),Weight=1))
    Weight=1
    bMaintainOffset=True
    ExecutePin=@n1.True
   ) {is=struct RigVMUnitNode}
   > n2 = &RigUnit_AimConstraintLocalSpaceOffset(
    AdvancedSettings=(DebugSettings=(bEnabled=false,Scale=10,WorldOffset=(Rotation=(X=0,Y=0,Z=0,W=1),Translation=(X=0,Y=0,Z=0),Scale3D=(X=1,Y=1,Z=1))),RotationOrderForFilter=XZY)
    Child=(Type=Null,Name="pivot_center")
    Filter=(bX=false,bY=true,bZ=true)
    Parents=((Item=(Type=Control,Name="pivot_top"),Weight=1))
    Weight=1
    WorldUp=(Target=(X=0,Y=0,Z=1),Kind=Direction,Space=(Type=None,Name="None"))
    bMaintainOffset=true
   )
   GetAnimationChannel = /Script/ControlRig.RigUnit_GetBoolAnimationChannel(Channel=Break Pivot, Control=root, bInitial=False) {is=struct RigVMUnitNode, template="GetAnimationChannel::Execute(out Value,in Control,in Channel,in bInitial)"}
   RigVMFunction_ControlFlowBranch = &RigVMFunction_ControlFlowBranch(ExecuteContext=@n2.ExecutePin, Condition=@GetAnimationChannel.Value)
   n3 = /Script/ControlRig.RigUnit_ProjectTransformToNewParent(
    Child=(Type=Null,Name="MainNull")
    NewParent=(Type=Null,Name="CenterNull")
    OldParent=(Type=Null,Name="CenterNull")
    bChildInitial=True
    bNewParentInitial=False
    bOldParentInitial=True
   ) {is=struct RigVMUnitNode}
   &RigUnit_SetTransform(Item=(Type=Null,Name="MainNull"), Weight=1, bInitial=False, Value=@n3.Transform, ExecutePin=@RigVMFunction_ControlFlowBranch.False)
   n4 = RigVMVariableNode() {variable=Control Array}
   &RigUnit_SetTransform(
    Weight=1
    bInitial=False
    Item=@For_Each.Element
    Value=&RigUnit_GetTransform(
     bInitial=False
     Item=RigVMDispatchNode(Index=@For_Each.Index, Array=@n4.Value) {template="DISPATCH_RigVMDispatch_ArrayGetAtIndex(in Array,in Index,out Element)"}.Element
    ).Transform
    ExecutePin=@n1.Completed
   )
   RigVMFunctionReturnNode(ExecuteContext=@RigVMFunction_Sequence.C)
   Multiply = &RigVMFunction_MathTransformMul(
    A=(Rotation=(X=0,Y=0,Z=0,W=1),Translation=(X=0,Y=0,Z=0),Scale3D=(X=1,Y=1,Z=1))
    B=(Rotation=(X=0,Y=0,Z=0,W=1),Translation=(X=0,Y=0,Z=0),Scale3D=(X=1,Y=1,Z=1))
   ) {template="Multiply::Execute(in A,in B,out Result)"}
   &RigUnit_ParentConstraint(
    AdvancedSettings=(InterpolationType=Average,RotationOrderForFilter=XZY)
    Child=(Type=Null,Name="MainNull")
    ExecutePin=()
    Parents=((Item=(Type=Null,Name="CenterNull"),Weight=1))
    Weight=1
    bMaintainOffset=false
   )
   n5 = &RigUnit_ProjectTransformToNewParent(
    Child=(Type=Control,Name="pivot_center")
    NewParent=(Type=Control,Name="pivot_top")
    OldParent=(Type=Control,Name="pivot_top")
    bChildInitial=True
    bNewParentInitial=False
    bOldParentInitial=True
   )
   n6 = &RigUnit_ProjectTransformToNewParent(
    Child=(Type=Control,Name="pivot_center")
    NewParent=(Type=Control,Name="pivot_bottom")
    OldParent=(Type=Control,Name="pivot_bottom")
    bChildInitial=True
    bNewParentInitial=False
    bOldParentInitial=True
   )
   Interpolate = /Script/RigVM.RigVMFunction_MathTransformLerp(T=0, A=@n5.Transform, B=@n6.Transform) {is=struct RigVMUnitNode, template="Interpolate::Execute(in A,in B,in T,out Result)"}
   &RigUnit_SetTransform(ExecutePin=(), Item=(Type=Control,Name="pivot_center"), Weight=1, bInitial=False, Value=@Interpolate.Result)
  ]}
 New Function_ContainedGraph {nodes [
   n1 = RigVMFunctionEntryNode()
   RigVMFunction_Sequence = &RigVMFunction_Sequence(ExecuteContext=@n1.ExecuteContext)
   GetAnimationChannel = &RigUnit_GetBoolAnimationChannel(Channel=Hide Pivot Controls, Control=root, bInitial=False) {template="GetAnimationChannel::Execute(out Value,in Control,in Channel,in bInitial)"}
   RigVMFunction_MathBoolNot = &RigVMFunction_MathBoolNot(Value=@GetAnimationChannel.Value)
   SetControlVisibility = /Script/ControlRig.RigUnit_SetControlVisibility(Item=(Type=Control,Name="pivot_bottom"), bVisible=@RigVMFunction_MathBoolNot.Result, ExecutePin=@RigVMFunction_Sequence.A) {is=struct RigVMUnitNode, template="SetControlVisibility::Execute(in Item,in Pattern,in bVisible)"}
   > &RigUnit_SetControlVisibility(Item=(Type=Control,Name="pivot_top"), bVisible=@RigVMFunction_MathBoolNot.Result) {template="SetControlVisibility::Execute(in Item,in Pattern,in bVisible)"}
   RigVMFunctionReturnNode(ExecuteContext=@RigVMFunction_Sequence.B)
  ]}
 RigVMFunctionLibrary {nodes [
   Animation Channels = RigVMCollapseNode(ExecuteContext=())
   Create Controls = RigVMCollapseNode(ExecuteContext=())
   FK solve = RigVMCollapseNode(ExecuteContext=())
   Import Skeleton = RigVMCollapseNode(ExecuteContext=())
   Reset Arrays = RigVMCollapseNode(ExecuteContext=())
   Visibility = RigVMCollapseNode(ExecuteContext=())
  ]}
 RigVMFunction_Sequence_1_ContainedGraph {nodes [
   n1 = RigVMFunctionEntryNode()
   RigVMFunction_Sequence = &RigVMFunction_Sequence(ExecuteContext=@n1.ExecuteContext)
   n2 = &RigVMFunction_Sequence(ExecuteContext=@RigVMFunction_Sequence.B)
   RigVMFunctionReturnNode(A=@RigVMFunction_Sequence.A, B=@n2.A, C=@n2.B)
  ]}
 RigVMFunction_Sequence_1_ContainedGraph {nodes [
   n1 = RigVMFunctionEntryNode()
   RigVMFunction_Sequence = &RigVMFunction_Sequence(ExecuteContext=@n1.ExecuteContext)
   n2 = &RigVMFunction_Sequence(ExecuteContext=@RigVMFunction_Sequence.B)
   RigVMFunctionReturnNode(A=@RigVMFunction_Sequence.A, B=@n2.A, C=@n2.B)
  ]}
 RigVMFunction_Sequence_2_ContainedGraph {nodes [
   n1 = RigVMFunctionEntryNode()
   n2 = &RigVMFunction_Sequence(ExecuteContext=@n1.ExecuteContext)
   n3 = &RigVMFunction_Sequence(ExecuteContext=@n2.B)
   n4 = &RigVMFunction_Sequence(ExecuteContext=@n3.B)
   n5 = &RigVMFunction_Sequence(ExecuteContext=@n4.B)
   RigVMFunctionReturnNode(A=@n2.A, B=@n3.A, C=@n4.A, D=@n5.A, E=@n5.B)
  ]}
 RigVMFunction_Sequence_3_ContainedGraph {nodes [
   n1 = RigVMFunctionEntryNode()
   n2 = &RigVMFunction_Sequence(ExecuteContext=@n1.ExecuteContext)
   n3 = &RigVMFunction_Sequence(ExecuteContext=@n2.B)
   RigVMFunctionReturnNode(B=@n3.A, C=@n3.B, A=@n2.A)
  ]}
 RigVMFunction_Sequence_4_ContainedGraph {nodes [
   n1 = RigVMFunctionEntryNode()
   RigVMFunction_Sequence = &RigVMFunction_Sequence(ExecuteContext=@n1.ExecuteContext)
   n2 = &RigVMFunction_Sequence(ExecuteContext=@RigVMFunction_Sequence.B)
   RigVMFunctionReturnNode(A=@RigVMFunction_Sequence.A, B=@n2.A, C=@n2.B)
  ]}
 RigVMModel {nodes [
   RigUnit_BeginExecution = &RigUnit_BeginExecution()
   FK solve = RigVMFunctionReferenceNode(ExecuteContext=@RigUnit_BeginExecution.ExecutePin)
   > Visibility = RigVMFunctionReferenceNode()
  ]}
 RigVMModel Rig {nodes [
   n1 = &RigUnit_PrepareForExecution()
   Reset Arrays = RigVMFunctionReferenceNode(ExecuteContext=@n1.ExecutePin)
   > Import Skeleton = RigVMFunctionReferenceNode()
   > Create Controls = RigVMFunctionReferenceNode()
   > Animation Channels = RigVMFunctionReferenceNode()
  ]}
 RigVMModel Rig_2 {nodes [
   n1 = /Script/ControlRig.RigUnit_InverseExecution() {is=struct RigVMUnitNode}
   For_Each = RigVMDispatchNode(Array=RigVMVariableNode() {variable=Control Array}.Value, ExecuteContext=@n1.ExecutePin) {template="DISPATCH_RigVMDispatch_ArrayIterator(in Array,out Element,out Index,out Count,out Ratio)"}
   RigVMFunction_ControlFlowBranch = &RigVMFunction_ControlFlowBranch(
    Condition=RigVMDispatchNode(B=2, A=@For_Each.Index) {template="DISPATCH_RigVMDispatch_CoreEquals(in A,in B,out Result)"}.Result
    ExecuteContext=@For_Each.ExecuteContext
   )
   n2 = RigVMVariableNode() {variable=Skeleton Array}
   &RigUnit_SetTransform(
    Weight=1
    bInitial=False
    bPropagateToChildren=false
    Item=@For_Each.Element
    Value=&RigUnit_GetTransform(
     bInitial=False
     Item=RigVMDispatchNode(Index=@For_Each.Index, Array=@n2.Value) {template="DISPATCH_RigVMDispatch_ArrayGetAtIndex(in Array,in Index,out Element)"}.Element
    ).Transform
    ExecutePin=@RigVMFunction_ControlFlowBranch.Completed
   )
   n3 = &RigUnit_GetTransform(Item=(Type=Bone,Name="offset"), bInitial=False)
   n4 = &RigUnit_ProjectTransformToNewParent(
    Child=(Type=Control,Name="pivot_top")
    NewParent=(Type=Bone,Name="main")
    OldParent=(Type=Bone,Name="main")
    bChildInitial=True
    bNewParentInitial=False
    bOldParentInitial=True
   )
   &RigUnit_SetTransform(
    Item=(Type=Control,Name="pivot_top")
    Value=(Rotation=(X=0,Y=0,Z=0,W=1),Translation=(X=0,Y=0,Z=0),Scale3D=(X=1,Y=1,Z=1))
    Weight=1
    bInitial=False
    Value.Rotation=@n3.Transform.Rotation
    Value.Translation=@n4.Transform.Translation
    ExecutePin=@RigVMFunction_ControlFlowBranch.True
   )
   n5 = &RigUnit_ProjectTransformToNewParent(
    Child=(Type=Control,Name="pivot_bottom")
    NewParent=(Type=Bone,Name="main")
    OldParent=(Type=Bone,Name="main")
    bChildInitial=True
    bNewParentInitial=False
    bOldParentInitial=True
   )
   > &RigUnit_SetTransform(
    Item=(Type=Control,Name="pivot_bottom")
    Value=(Rotation=(X=0,Y=0,Z=0,W=1),Translation=(X=0,Y=0,Z=0),Scale3D=(X=1,Y=1,Z=1))
    Weight=1
    bInitial=False
    Value.Rotation=@n3.Transform.Rotation
    Value.Translation=@n5.Transform.Translation
   )
   n6 = &RigUnit_ProjectTransformToNewParent(
    Child=(Type=Null,Name="pivot_center")
    NewParent=(Type=Bone,Name="main")
    OldParent=(Type=Bone,Name="main")
    bChildInitial=True
    bNewParentInitial=False
    bOldParentInitial=True
   )
   > &RigUnit_SetTransform(Item=(Type=Null,Name="pivot_center"), Weight=1, bInitial=False, bPropagateToChildren=true, Value=@n6.Transform)
   n7 = &RigUnit_ProjectTransformToNewParent(
    Child=(Type=Null,Name="MainNull")
    NewParent=(Type=Null,Name="CenterNull")
    OldParent=(Type=Null,Name="CenterNull")
    bChildInitial=True
    bNewParentInitial=False
    bOldParentInitial=True
   )
   > &RigUnit_SetTransform(Item=(Type=Null,Name="MainNull"), Weight=1, bInitial=False, Value=@n7.Transform)
   &RigUnit_PositionConstraintLocalSpaceOffset(
    Child=(Type=Control,Name="pivot_center")
    ExecutePin=()
    Filter=(bX=True,bY=True,bZ=True)
    Parents=((Item=(Type=Control,Name="pivot_top"),Weight=1),(Item=(Type=Control,Name="pivot_bottom"),Weight=1))
    Weight=1
    bMaintainOffset=True
   )
   > &RigUnit_AimConstraintLocalSpaceOffset(
    AdvancedSettings=(DebugSettings=(bEnabled=false,Scale=10,WorldOffset=(Rotation=(X=0,Y=0,Z=0,W=1),Translation=(X=0,Y=0,Z=0),Scale3D=(X=1,Y=1,Z=1))),RotationOrderForFilter=XZY)
    Child=(Type=Control,Name="pivot_center")
    Filter=(bX=false,bY=true,bZ=true)
    Parents=((Item=(Type=Control,Name="pivot_top"),Weight=1))
    Weight=1
    WorldUp=(Target=(X=0,Y=0,Z=1),Kind=Direction,Space=(Type=None,Name="None"))
    bMaintainOffset=true
   )
   > &RigUnit_ParentConstraint(
    AdvancedSettings=(InterpolationType=Average,RotationOrderForFilter=XZY)
    Child=(Type=Null,Name="MainNull")
    Parents=((Item=(Type=Null,Name="CenterNull"),Weight=1))
    Weight=1
    bMaintainOffset=true
   )
   &RigUnit_GetTransform(Item=(Type=Bone,Name="main"), bInitial=False)
   &RigUnit_ParentConstraint(
    AdvancedSettings=(InterpolationType=Average,RotationOrderForFilter=XZY)
    Child=(Type=Control,Name="pivot_center")
    ExecutePin=()
    Parents=((Item=(Type=Bone,Name="main"),Weight=1))
    Weight=1
    bMaintainOffset=false
   )
  ]}
]
hierarchy [
 root {loc=<0 0 0>, rot=<0 0 0 1>}
 offset {parent=root, loc=<0 0 0>, rot=<0 0 -0.7071 -0.7071>}
 main {parent=offset, loc=<-0.0019 0.0002 -198.3217>, rot=<0 0 0 1>}
 bay_door_outer_l {parent=main, loc=<0 0 0>, rot=<0 0 1 0>}
 bay_door_inner_l {parent=main, loc=<0 0 0>, rot=<0 0 1 0>}
 bay_door_outer_r {parent=main, loc=<0 0 0>, rot=<0 0 0 1>}
 bay_door_inner_r {parent=main, loc=<0 0 0>, rot=<0 0 0 1>}
 cannon_apparatus {parent=main, loc=<0 0 0>, rot=<0 -0.7071 0 -0.7071>}
 cannon_main {parent=cannon_apparatus, loc=<3686.7031 0.0007 -0.0004>, rot=<0 0 0 1>}
 cannon_arm_A {parent=cannon_main, loc=<-840.6849 -623.4297 673.9682>, rot=<-0.3302 0.3833 0.149 -0.8496>}
 cannon_arm_B {parent=cannon_main, loc=<-840.6849 -836.9563 -377.3589>, rot=<0.7644 -0.224 -0.3448 0.4966>}
 cannon_arm_C {parent=cannon_main, loc=<-840.6849 100.2559 -912.6037>, rot=<0.9103 0.0214 -0.4106 -0.0475>}
 cannon_arm_D {parent=cannon_main, loc=<-840.6849 898.9211 -186.6589>, rot=<0.7085 0.2587 -0.3196 -0.5735>}
 cannon_arm_E {parent=cannon_main, loc=<-840.6848 455.3068 797.2414>, rot=<0.2361 0.3972 -0.1065 -0.8804>}
 bay_door_fr {parent=main, loc=<0 0 0>, rot=<0 0 0.7071 -0.7071>}
 bay_door_bk {parent=main, loc=<0 0 0>, rot=<0 0 -0.7071 -0.7071>}
]
variables [
 Anim Channel Array {type="TArray<FRigElementKey>"}
 Control Array {type="TArray<FRigElementKey>"}
 Index {type=int32, default=0}
 Skeleton Array {type="TArray<FRigElementKey>"}
]
